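; Quest ability for an "eliminate after elimination" objective, run from a gameplay event.
; After the parent logic, it requires the event target to match any include tag and no exclude tag.
; If the owning player's pawn has an instigator include tag, it sends a custom stat event to that player's quest manager.
; Otherwise it applies a configured gameplay effect to the player's controller, apparently priming the next qualifying elimination.
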(event Ability|EventActivateAbilityFromEvent
  (params (EventData :type Gameplay%20Event%20Data%20Structure%20%28by%20ref%29))
  (bind (_eventtag _instigator _target) (Utilities|Struct|BreakGameplayEventData EventData))
  (|CallParentFunction EventData)
  (bind _returnvalue (GameplayTags|HasAnyMatchingGameplayTags(Message) _target (Variables|Default|GetTargetIncludeTagContainer)))
  (if _returnvalue
    (bind _returnvalue_2 (GameplayTags|HasAnyMatchingGameplayTags(Message) _target (Variables|Default|GetTargetExcludeTagContainer)))
    (if (not _returnvalue_2)
      (bind _returnvalue_3 (GameplayTags|HasAnyMatchingGameplayTags(Message) (Variables|GAVars|Data|GetOwningPlayerPawn) (Variables|Default|GetInstigatorIncludeTagContainer)))
      (if _returnvalue_3
        (Quest|SendComplexCustomStatEvent (Variables|GAVars|Data|GetOwningPlayerQuestManager) 0 (Variables|Default|GetAdditionalSourceTags) "(GameplayTags=)" 1)
        (else
          (Creative|GameplayEffects|SimpleApplyGameplayEffectToActor (Variables|Default|GetGameplay%20Effect) (Fortnite|GetFortPlayerControllerfromActor (Variables|GAVars|Data|GetOwningPlayerPawn))))))))
